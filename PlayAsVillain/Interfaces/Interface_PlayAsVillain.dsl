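; Blueprint interface for the Play As Villain feature, declaring three messages with no implementation in this file.
; PlayAsVillainAbilityUsed and PlayAsVillainAIBossEliminated take no parameters and, by name, signal ability use and AI boss elimination.
; SpawnedForPlayAsVillain passes the instigating Fort player pawn and a float duration to the implementer.
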(fn PlayAsVillainAbilityUsed)

(fn PlayAsVillainAIBossEliminated)

(fn SpawnedForPlayAsVillain
  (params (Instigator :type Fort%20Player%20Pawn%20Object%20Reference) (Duration :type Float%20%28double-precision%29)))
